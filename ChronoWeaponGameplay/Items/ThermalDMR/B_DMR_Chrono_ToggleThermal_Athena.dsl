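; Thermal DMR weapon Blueprint whose scope shows toggleable thermal vision while aiming in target mode 1.
; Targeting, mode changes and EndPlay call EnableThermalVision, where authority applies or removes a GameplayEffect and triggers cues.
; The local viewer predictively gets a loose thermal tag and forced custom-depth updates to avoid glitches.
; Timelines animate thermal and reticle scope materials, and locally fired projectiles are flagged as predicted.
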(event EventOnSetTargeting
  (params (bNewIsTargeting :type Boolean))
  ;[1 1 1] Comment
  (|CallParentFunction bNewIsTargeting)
  (CallFunction|EnableThermalVision (CallFunction|IsThermalVisionOn)))

(event OnTargetModeChanged(TargetModeSelector)
  (params (TargetMode :type Integer))
  ;[1 1 1] Comment
  (CallFunction|EnableThermalVision (CallFunction|IsThermalVisionOn)))

(event Custom|Update%20Scope%20Post%20Process
  (bind _target_mode_selector (Variables|Default|GetTarget%20Mode%20Selector))
  (bind _targetmode (Variables|/Script/FortniteGame.FortWeaponComponent_TargetModeSelector|GetTargetMode _target_mode_selector))
  (bind _thermalmid (Variables|Default|GetThermalMID))
  ;[1 1 1] Comment
  (CallFunction|CacheMidVariables)
  (bind (_direction _togglethermalscale _togglemaskscale) (Timeline|Timeline PlayFromStart)
    (:Update
      _)
    (:Finished
      _))
  (Rendering|Material|SetScalarParameterValuebyInfo _thermalmid (Variables|Default|GetToggleThermalScale) _togglethermalscale)
  (Rendering|Material|SetScalarParameterValuebyInfo _thermalmid (Variables|Default|GetToggleThermalMask) (Math|Float|Max(Float) (Math|Conversions|ToFloat(Integer) _targetmode) _togglemaskscale))
  (Rendering|Material|SetVectorParameterValue (Variables|Default|GetReticleMID) "Reticle Color_Center" (Math|Color|Lerp(LinearColor) (Math|Color|MakeColor 0.0) (Math|Color|MakeColor 1.0) (Utilities|Select :Index _targetmode :Option%200 _togglemaskscale :Option%201 (- 1.0 _togglemaskscale))) :!class /Script/Engine.MaterialInstanceDynamic))

(event EventPlayScopePP
  (bind _thermalmid (Variables|Default|GetThermalMID))
  ;[1 1 1] Manually Enabling/Disabling PP so player does not see himself through Thermal Scope
  (CallFunction|CacheMidVariables)
  (Utilities|FlowControl|RetriggerableDelay 0.1)
  (if (Camera|GetIsTargeting)
    (CallFunction|SetPostProcessParams 1.0)
    (Components|Timeline|SetPlayRate (Variables|B_DMR_Chrono_ToggleThermal_Athena|GetScalePlayerThermal) (/ 1.0 (Variables|Default|GetBlendInTime)))
    (Utilities|FlowControl|DoOnce false
      (:Completed
        (bind (_direction _newtrack_0) (Timeline|ScalePlayerThermal PlayFromStart)
          (:Update
            (Rendering|Material|SetScalarParameterValuebyInfo _thermalmid (Variables|Default|GetThermalScale) _newtrack_0)
            (Rendering|Material|SetScalarParameterValuebyInfo _thermalmid (Variables|Default|GetToggleThermalScale) _newtrack_0)))))))

(event EventReverseScopePP
  ;[1 1 1] Manually Enabling/Disabling PP so player does not see himself through Thermal Scope
  ;;[0.1918 1 0.1475] No longer being called by parent
  (|CallParentFunction))

(event EventForceScopeBackImmediatly
  (bind _thermalmid (Variables|Default|GetThermalMID))
  ;[1 1 1] Manually Enabling/Disabling PP so player does not see himself through Thermal Scope
  (|CallParentFunction)
  (Utilities|FlowControl|Sequence
    (:then_0
      (Utilities|FlowControl|DoOnce false
        (:Completed
          (Timeline|ScalePlayerThermal PlayFromStart))))
    (:then_1
      (Timeline|ScalePlayerThermal Stop))))

(event FortWeapon|EventOnSpawnProjectile
  (params (SpawnProjectile :type Fort%20Projectile%20Base%20Object%20Reference))
  ;[1 1 1] Comment
  (if (Pawn|IsLocallyControlled (Variables|Actor|GetInstigator))
    (bind _asb_prj_bullet_toggle_thermal_dmr (Utilities|Casting|CastToB_Prj_Bullet_ToggleThermalDMR SpawnProjectile))
    (Variables|/ChronoWeaponGameplay/Items/ThermalDMR/B_Prj_Bullet_ToggleThermalDMR.B_Prj_Bullet_ToggleThermalDMR_C|SetIsPredicted true _asb_prj_bullet_toggle_thermal_dmr)))

(event Custom|Play%20ToggleThermalOFF%20Local
  (bind _returnvalue (Ability|GameplayCue|MakeGameplayCueParameters 0.0))
  (GameplayCue|ExecuteGameplayCue%28Burst%29Local (Game|GetInstigator) (Variables|Default|GetGameplay%20Cue%20Tag) _returnvalue)
  (Projectile|RemoveGameplayCue%28Looping%29Local (Game|GetInstigator) (Variables|Default|GetGameplay%20Cue%20Tag_0) _returnvalue))

(event Custom|Play%20ToggleThermal%20OnLoop
  (Projectile|AddGameplayCue%28Looping%29Local (Game|GetInstigator) (Variables|Default|GetGameplay%20Cue%20Tag_0) (Ability|GameplayCue|MakeGameplayCueParameters 0.0)))

(event EventEndPlay
  (params (EndPlayReason :type EEndPlayReason%20Enum :default "Destroyed"))
  ;[1 1 1 1 14] Make sure gameplay tags, custom stencil override, etc get cleaned up.
  (CallFunction|EnableThermalVision false))

(fn CacheMidVariables
  (bind (_weight _object) (Utilities|Struct|BreakWeightedBlendable (Utilities|Array|Get(aref) (Scope|GetScopePostProcessMaterial))))
  (bind (_weight_2 _object_2) (Utilities|Struct|BreakWeightedBlendable (Utilities|Array|Get(aref) (Scope|GetScopePostProcessMaterial) 1)))
  (Utilities|FlowControl|Sequence
    (:then_0
      (bind _asmaterial_instance_dynamic (Utilities|Casting|CastToMaterialInstanceDynamic _object_2))
      (bind _output_get (Variables|Default|SetThermalMID _asmaterial_instance_dynamic))
      (Variables|Default|SetToggleThermalScale (Rendering|Material|GetParameterInfo _output_get "LayerParameter" "ToggleThermalScale" "/Game/Weapons/FORT_Rifles/ThermalVision/ThermalAnimation/ML_AssaultSurgicalThermal_Toggle.ML_AssaultSurgicalThermal_Toggle"))
      (Variables|Default|SetToggleThermalMask (Rendering|Material|GetParameterInfo _output_get "LayerParameter" "ToggleThermalMask" "/Game/Weapons/FORT_Rifles/ThermalVision/ThermalAnimation/ML_AssaultSurgicalThermal_Toggle.ML_AssaultSurgicalThermal_Toggle")))
    (:then_1
      (bind _asmaterial_instance_dynamic_2 (Utilities|Casting|CastToMaterialInstanceDynamic _object))
      (Variables|Default|SetReticleMID _asmaterial_instance_dynamic_2))))

(fn EnableThermalVision
  (params (Enable :type Boolean))
  (bind _returnvalue (CallFunction|GetLocallyViewingPlayerController))
  (CallFunction|UpdateThermalVisionGE)
  (Utilities|IsValid _returnvalue
    (:Is%20Valid
      (CallFunction|UpdateLooseThermalVisionTag Enable)
      (CallFunction|UpdateScopePostProcess)
      (bind _asfort_player_controller (Utilities|Casting|CastToFortPlayerController _returnvalue))
      ;[0.3515 0.9131 0.2462 1 14] This forces wildlife to all render to custom depth, instead of selectively rendering to custom depth based on significance.
      (Rendering|SetIgnoreSignifanceBasedCustomDepthRendering _asfort_player_controller Enable 0)
      ;[0.3515 0.9131 0.2462 1 14] Force immediate update to custom depth stencil to prevent visual glitches when enabling thermal vision. Normally, these updates are throttled to 400 ms.
      (Rendering|ForceRenderCustomDepthUpdateforAllActors))))

(fn GetLocallyViewingPlayerController
  (Networking|CanExecuteCosmeticEvents
    (:True
      (return (Pawn|GetLocalViewingPlayerController (Game|GetInstigator))))))

(fn IsThermalVisionOn
  (return (and (Camera|GetIsTargeting) (== (Variables|/Script/FortniteGame.FortWeaponComponent_TargetModeSelector|GetTargetMode (Variables|Default|GetTarget%20Mode%20Selector)) 1))))

(fn UpdateThermalVisionGE
  (bind _instigator (Variables|Actor|GetInstigator))
  (bind _returnvalue (Ability|GetFortAbilitySystemComponent _instigator))
  (bind _activegameplayeffecthandle (Utilities|Struct|MakeActiveGameplayEffectHandle))
  (bind _thermalvisionge (Variables|Default|GetThermalVisionGE))
  (bind _returnvalue_2 (!= _thermalvisionge _activegameplayeffecthandle))
  (if (Networking|HasAuthority (Variables|Actor|GetInstigator))
    (if (CallFunction|IsThermalVisionOn)
      (if (not _returnvalue_2)
        (bind _returnvalue_3 (GameplayEffects|ApplyGameplayEffectToSelf :self _returnvalue :GameplayEffectClass "/ChronoWeaponGameplay/Items/ThermalDMR/GE_ThermalVision_ToggleThermalDMR.GE_ThermalVision_ToggleThermalDMR_C" :Level 0.0))
        (Variables|Default|SetThermalVisionGE _returnvalue_3)
        (CallFunction|PlayToggleThermalonLoop))
      (elif _returnvalue_2
        (GameplayEffects|RemoveActiveGameplayEffect _returnvalue (Variables|Default|GetThermalVisionGE) -1)
        (Variables|Default|SetThermalVisionGE (Utilities|Struct|MakeActiveGameplayEffectHandle))
        (CallFunction|PlayToggleThermalOFFLocal)))))

(fn UpdateLooseThermalVisionTag
  (params (ThermalEnabled :type Boolean))
  (bind _returnvalue (GameplayTags|MakeLiteralGameplayTagContainer "(GameplayTags=((TagName=\"Gameplay.Status.ThermalVision\")))"))
  (bind _returnvalue_2 (CallFunction|GetLocallyViewingPlayerController))
  (bind _hasaddedthermalvisiontag (Variables|Default|GetHasAddedThermalVisionTag))
  ;[0.3515 0.9131 0.2462 1 14] Predictively add Thermal.Vision tag to local player controller to prevent visual glitches while waiting for the server applied GE.
  (if ThermalEnabled
    (if (not _hasaddedthermalvisiontag)
      (bind _returnvalue_3 (Ability|GameplayEffect|AddLooseGameplayTags _returnvalue_2 _returnvalue))
      (Variables|Default|SetHasAddedThermalVisionTag _returnvalue_3))
    (elif _hasaddedthermalvisiontag
      (Ability|GameplayEffect|RemoveLooseGameplayTags _returnvalue_2 _returnvalue)
      (Variables|Default|SetHasAddedThermalVisionTag false))))

(fn ConstructionScript
  (|CallParentFunction))
