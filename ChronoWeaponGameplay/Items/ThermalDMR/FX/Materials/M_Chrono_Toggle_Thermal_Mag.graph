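# Default Lit surface material for the Thermal DMR weapon's magazine effect, flagged for use on Niagara mesh particles.
# It samples the Chrono Toggle Thermal texture set: diffuse (D), normal (N), and two mask textures (S and CM).
# The graph holds only these four samples, with no math, parameters or visible connections to material outputs.
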Material /ChronoWeaponGameplay/Items/ThermalDMR/FX/Materials/M_Chrono_Toggle_Thermal_Mag
graph {
 properties {bUsedWithNiagaraMeshParticles=true, pixelDepthOffsetMode=PDOM_Legacy, usedShadingModels=Default Lit}
 nodes [
  TextureSample() {samplerType=SAMPLERTYPE_Masks, texture=/ChronoWeaponGameplay/Items/ThermalDMR/Textures/T_Chrono_Toggle_Thermal_S}
  TextureSample() {samplerType=SAMPLERTYPE_Masks, texture=/ChronoWeaponGameplay/Items/ThermalDMR/Textures/T_Chrono_Toggle_Thermal_CM}
  TextureSample() {texture=/ChronoWeaponGameplay/Items/ThermalDMR/Textures/T_Chrono_Toggle_Thermal_D}
  TextureSample() {samplerType=SAMPLERTYPE_Normal, texture=/ChronoWeaponGameplay/Items/ThermalDMR/Textures/T_Chrono_Toggle_Thermal_N}
 ]
}
