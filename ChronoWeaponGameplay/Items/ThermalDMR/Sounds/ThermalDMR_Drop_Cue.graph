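# Sound cue for the drop sound of the ThermalDMR weapon, classed as loot-collect audio.
# A random node picks one of two equally weighted drop waves, and a modulator slightly varies its pitch and volume.
# Playback uses the shared weapon-drop attenuation and the Weapon_Drop_StopOldest concurrency asset to manage overlapping drops.
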SoundCue /ChronoWeaponGameplay/Items/ThermalDMR/Sounds/ThermalDMR_Drop_Cue
root=Modulator_0
nodes [
 {id=Modulator_0, children [Random_0], properties {pitchMax=1.1, pitchMin=0.9, volumeMax=1.1, volumeMin=0.9}}
 {id=Random_0, children [WavePlayer_0,WavePlayer_2], properties {weights [1,1]}}
 {id=WavePlayer_0, properties {soundWaveAssetPtr=&/ThermalDMR_Drop_01}}
 {id=WavePlayer_2, properties {soundWaveAssetPtr=&/ThermalDMR_Drop_02}}
]
properties {
 attenuationSettings=&Fort_Weapon_Drop_Att
 maxDistance=3000
 soundClassObject=&Fort_Loot_Collect
 concurrencyOverrides {resolutionRule=PreventNew}
 concurrencySet [&Weapon_Drop_StopOldest]
}
