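# ThermalDMR_Drop_Impact_Cue is a SoundCue that plays the impact sound when the Thermal DMR weapon is dropped.
# A Random node picks evenly among three impact waves, then a Modulator adds slight pitch variation and a boosted, varied volume.
# The cue uses the shared weapon-drop attenuation and concurrency set and the loot-collect sound class.
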SoundCue /ChronoWeaponGameplay/Items/ThermalDMR/Sounds/ThermalDMR_Drop_Impact_Cue
root=Modulator_0
nodes [
 {id=Modulator_0, children [Random_1], properties {pitchMax=1.1, pitchMin=0.9, volumeMax=1.4, volumeMin=1.2}}
 {id=Random_1, children [WavePlayer_3,WavePlayer_4,WavePlayer_5], properties {weights [1,1,1]}}
 {id=WavePlayer_3, properties {soundWaveAssetPtr=&/ThermalDMR_Drop_Impact_01}}
 {id=WavePlayer_4, properties {soundWaveAssetPtr=&/ThermalDMR_Drop_Impact_02}}
 {id=WavePlayer_5, properties {soundWaveAssetPtr=&/ThermalDMR_Drop_Impact_03}}
]
properties {
 attenuationSettings=&Fort_Weapon_Drop_Att
 maxDistance=3000
 soundClassObject=&Fort_Loot_Collect
 volumeMultiplier=0.8
 concurrencyOverrides {resolutionRule=PreventNew}
 concurrencySet [&Weapon_Drop_StopOldest]
}
