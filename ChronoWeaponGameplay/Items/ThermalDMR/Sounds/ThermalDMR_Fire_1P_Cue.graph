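# First-person fire sound cue for the Thermal DMR, with first-person attenuation and gunfire concurrency limited per owner, stopping the oldest.
# A quality-level root sends its first two slots to a full mix of randomized shot, slightly delayed mechanical layer, fixed sub layer and later shell impact.
# The last two slots use a lighter mix with fewer variants, a different mechanical set with volume modulation, and no sub layer.
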SoundCue /ChronoWeaponGameplay/Items/ThermalDMR/Sounds/ThermalDMR_Fire_1P_Cue
root=QualityLevel_0
nodes [
 {id=Delay_0, children [Modulator_0], properties {delayMax=0.05, delayMin=0.05}}
 {id=Delay_1, children [Random_5], properties {delayMax=0.5, delayMin=0.5}}
 {id=Delay_2, children [Modulator_2], properties {delayMax=0.07, delayMin=0.07}}
 {id=Delay_3, children [Random_8], properties {delayMax=0.5, delayMin=0.5}}
 {id=Mixer_0, children [Modulator_1, Delay_0, WavePlayer_26, Delay_1], properties {inputVolume [1,0.4,0.7,0.25]}}
 {id=Mixer_2, children [Modulator_3,Delay_2,Delay_3], properties {inputVolume [1,0.4,0.25]}}
 {id=Modulator_0, children [Random_1]}
 {id=Modulator_1, children [Random_0]}
 {id=Modulator_2, children [Random_7], properties {pitchMax=1, pitchMin=1, volumeMax=1.1, volumeMin=0.8}}
 {id=Modulator_3, children [Random_6]}
 {id=QualityLevel_0, children [Mixer_0,Mixer_0,Mixer_2,Mixer_2]}
 {id=Random_0, children [WavePlayer_2,WavePlayer_3,WavePlayer_4,WavePlayer_5,WavePlayer_7], properties {weights [1,1,1,1,1]}}
 {id=Random_1, children [WavePlayer_0,WavePlayer_1,WavePlayer_6,WavePlayer_8,WavePlayer_9], properties {weights [1,1,1,1,1]}}
 {
  id=Random_5
  children [WavePlayer_28, WavePlayer_29, WavePlayer_30, WavePlayer_31, WavePlayer_32, WavePlayer_33]
  properties {weights [1,1,1,1,1,1]}
 }
 {id=Random_6, children [WavePlayer_34, WavePlayer_35, WavePlayer_38], properties {weights [1,1,1]}}
 {id=Random_7, children [WavePlayer_40, WavePlayer_43], properties {weights [1,1]}}
 {id=Random_8, children [WavePlayer_45, WavePlayer_46, WavePlayer_47], properties {weights [1,1,1]}}
 {id=WavePlayer_0, properties {soundWaveAssetPtr=&/ThermalDMR_Fire_Close_MechB_01}}
 {id=WavePlayer_1, properties {soundWaveAssetPtr=&/ThermalDMR_Fire_Close_MechB_02}}
 {id=WavePlayer_2, properties {soundWaveAssetPtr=&/ThermalDMR_Fire_Close_A_01}}
 {id=WavePlayer_26, properties {soundWaveAssetPtr=&/ThermalDMR_Fire_Sub_03}}
 {id=WavePlayer_28, properties {soundWaveAssetPtr=&/ThermalDMR_Shell_Impact_01}}
 {id=WavePlayer_29, properties {soundWaveAssetPtr=&/ThermalDMR_Shell_Impact_02}}
 {id=WavePlayer_3, properties {soundWaveAssetPtr=&/ThermalDMR_Fire_Close_A_02}}
 {id=WavePlayer_30, properties {soundWaveAssetPtr=&/ThermalDMR_Shell_Impact_03}}
 {id=WavePlayer_31, properties {soundWaveAssetPtr=&/ThermalDMR_Shell_Impact_04}}
 {id=WavePlayer_32, properties {soundWaveAssetPtr=&/ThermalDMR_Shell_Impact_05}}
 {id=WavePlayer_33, properties {soundWaveAssetPtr=&/ThermalDMR_Shell_Impact_06}}
 {id=WavePlayer_34, properties {soundWaveAssetPtr=&ThermalDMR_Fire_Close_A_01}}
 {id=WavePlayer_35, properties {soundWaveAssetPtr=&ThermalDMR_Fire_Close_A_02}}
 {id=WavePlayer_38, properties {soundWaveAssetPtr=&/ThermalDMR_Fire_Close_A_05}}
 {id=WavePlayer_4, properties {soundWaveAssetPtr=&/ThermalDMR_Fire_Close_A_03}}
 {id=WavePlayer_40, properties {soundWaveAssetPtr=&/ThermalDMR_Fire_Close_Mech_02}}
 {id=WavePlayer_43, properties {soundWaveAssetPtr=&/ThermalDMR_Fire_Close_Mech_05}}
 {id=WavePlayer_45, properties {soundWaveAssetPtr=&ThermalDMR_Shell_Impact_01}}
 {id=WavePlayer_46, properties {soundWaveAssetPtr=&ThermalDMR_Shell_Impact_02}}
 {id=WavePlayer_47, properties {soundWaveAssetPtr=&ThermalDMR_Shell_Impact_03}}
 {id=WavePlayer_5, properties {soundWaveAssetPtr=&/ThermalDMR_Fire_Close_A_04}}
 {id=WavePlayer_6, properties {soundWaveAssetPtr=&/ThermalDMR_Fire_Close_MechB_03}}
 {id=WavePlayer_7, properties {soundWaveAssetPtr=&ThermalDMR_Fire_Close_A_05}}
 {id=WavePlayer_8, properties {soundWaveAssetPtr=&/ThermalDMR_Fire_Close_MechB_04}}
 {id=WavePlayer_9, properties {soundWaveAssetPtr=&/ThermalDMR_Fire_Close_MechB_05}}
]
properties {
 attenuationSettings=&Fort_Weapon_Projectile_FirstPerson_Athena_Att
 bDebug=true
 bExcludeFromRandomNodeBranchCulling=true
 maxDistance=2097152
 soundClassObject=&Fort_WeaponRanged_Fire_1P_PSM
 volumeMultiplier=1.15
 concurrencyOverrides {bLimitToOwner=true, maxCount=3, resolutionRule=StopOldest}
 concurrencySet [&Weapon_Gunfire_FirstPerson_Conc]
}
