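# Sound cue for the Thermal DMR's gunshot, apparently the third-person version, layered by distance.
# A quality switch selects between two near-identical trees, each mixing close, medium and distant layers with separate attenuation and sound classes.
# The close layer blends a random shot, a slightly delayed mechanical sound and a fixed sub, while medium and distant layers pick random, pitch-modulated variants.
# Local and global gunfire concurrency sets govern simultaneous voices.
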SoundCue /ChronoWeaponGameplay/Items/ThermalDMR/Sounds/ThermalDMR_Fire_Cue
root=QualityLevel_1
nodes [
 {
  id=Attenuation_0
  children [Modulator_2]
  properties {attenuationSettings=/Game/Sounds/Attenuation/Fort_Weapon_Projectile_Close_Athena_Att_Scoped}
 }
 {
  id=Attenuation_1
  children [Modulator_3]
  properties {attenuationSettings=&Fort_Weapon_Projectile_Medium_Mono_Athena_Att}
 }
 {
  id=Attenuation_2
  children [Modulator_4]
  properties {attenuationSettings=&Fort_Weapon_Projectile_Distant_Mono_Athena_Att}
 }
 {
  id=Attenuation_3
  children [Modulator_8]
  properties {attenuationSettings=&Fort_Weapon_Projectile_Close_Athena_Att_Scoped}
 }
 {
  id=Attenuation_4
  children [Modulator_9]
  properties {attenuationSettings=&Fort_Weapon_Projectile_Medium_Mono_Athena_Att}
 }
 {
  id=Attenuation_5
  children [Modulator_10]
  properties {attenuationSettings=&Fort_Weapon_Projectile_Distant_Mono_Athena_Att}
 }
 {id=Delay_0, children [Modulator_6], properties {delayMax=0.07, delayMin=0.07}}
 {id=Delay_1, children [Modulator_0], properties {delayMax=0.07, delayMin=0.07}}
 {id=Mixer_0, children [SoundClass_0,Mixer_2], properties {inputVolume [1,0.65]}}
 {id=Mixer_1, children [SoundClass_2,Mixer_4], properties {inputVolume [1,0.65]}}
 {id=Mixer_2, children [SoundClass_1,SoundClass_8], properties {inputVolume [0.65,0.25]}}
 {id=Mixer_3, children [Modulator_1, Delay_1, WavePlayer_15], properties {inputVolume [0.9,0.4,0.7]}}
 {id=Mixer_4, children [SoundClass_3,SoundClass_4], properties {inputVolume [0.65,0.25]}}
 {id=Mixer_5, children [Modulator_7, Delay_0, WavePlayer_40], properties {inputVolume [0.9,0.4,0.7]}}
 {id=Modulator_0, children [Random_2], properties {pitchMax=1, pitchMin=1, volumeMax=1.1, volumeMin=0.8}}
 {id=Modulator_1, children [Random_1]}
 {id=Modulator_10, children [Random_12], properties {pitchMax=1, pitchMin=0.9, volumeMax=2, volumeMin=2}}
 {id=Modulator_2, children [Mixer_3], properties {volumeMax=1.1, volumeMin=1.1}}
 {id=Modulator_3, children [Random_4], properties {pitchMax=1.2, pitchMin=1.1, volumeMax=2, volumeMin=2}}
 {id=Modulator_4, children [Random_5], properties {pitchMax=1, pitchMin=0.9, volumeMax=2, volumeMin=2}}
 {id=Modulator_6, children [Random_3], properties {pitchMax=1, pitchMin=1, volumeMax=1.1, volumeMin=0.8}}
 {id=Modulator_7, children [Random_0]}
 {id=Modulator_8, children [Mixer_5], properties {volumeMax=1.1, volumeMin=1.1}}
 {id=Modulator_9, children [Random_11], properties {pitchMax=1.2, pitchMin=1.1, volumeMax=2, volumeMin=2}}
 {id=QualityLevel_1, children [Mixer_1,Mixer_1,Mixer_0,Mixer_0]}
 {id=Random_0, children [WavePlayer_2,WavePlayer_3,WavePlayer_4,WavePlayer_5,WavePlayer_7], properties {weights [1,1,1,1,1]}}
 {id=Random_1, children [WavePlayer_0,WavePlayer_1,WavePlayer_6,WavePlayer_8,WavePlayer_9], properties {weights [1,1,1,1,1]}}
 {id=Random_11, children [WavePlayer_45, WavePlayer_46, WavePlayer_47, WavePlayer_48, WavePlayer_49], properties {weights [1,1,1,1,1]}}
 {id=Random_12, children [WavePlayer_50, WavePlayer_51, WavePlayer_52, WavePlayer_53, WavePlayer_54], properties {weights [1,1,1,1,1]}}
 {id=Random_2, children [WavePlayer_10, WavePlayer_11, WavePlayer_12, WavePlayer_13, WavePlayer_14], properties {weights [1,1,1,1,1]}}
 {id=Random_3, children [WavePlayer_35, WavePlayer_36, WavePlayer_37, WavePlayer_38, WavePlayer_39], properties {weights [1,1,1,1,1]}}
 {id=Random_4, children [WavePlayer_16, WavePlayer_17, WavePlayer_18, WavePlayer_19, WavePlayer_20], properties {weights [1,1,1,1,1]}}
 {id=Random_5, children [WavePlayer_21, WavePlayer_22, WavePlayer_23, WavePlayer_24, WavePlayer_25], properties {weights [1,1,1,1,1]}}
 {
  id=SoundClass_0
  children [Attenuation_0]
  properties {soundClassOverride=&Fort_WeaponRanged_Fire_3P_PSM}
 }
 {
  id=SoundClass_1
  children [Attenuation_1]
  properties {soundClassOverride=&Fort_WeaponRanged_Fire_Distant}
 }
 {
  id=SoundClass_2
  children [Attenuation_3]
  properties {soundClassOverride=&Fort_WeaponRanged_Fire_3P_PSM}
 }
 {
  id=SoundClass_3
  children [Attenuation_4]
  properties {soundClassOverride=&Fort_WeaponRanged_Fire_Distant}
 }
 {
  id=SoundClass_4
  children [Attenuation_5]
  properties {soundClassOverride=&Fort_WeaponRanged_Fire_Distant}
 }
 {
  id=SoundClass_8
  children [Attenuation_2]
  properties {soundClassOverride=&Fort_WeaponRanged_Fire_Distant}
 }
 {id=WavePlayer_0, properties {soundWaveAssetPtr=&/ThermalDMR_Fire_Close_A_01}}
 {id=WavePlayer_1, properties {soundWaveAssetPtr=&/ThermalDMR_Fire_Close_A_02}}
 {id=WavePlayer_10, properties {soundWaveAssetPtr=&/ThermalDMR_Fire_Close_Mech_01}}
 {id=WavePlayer_11, properties {soundWaveAssetPtr=&/ThermalDMR_Fire_Close_Mech_02}}
 {id=WavePlayer_12, properties {soundWaveAssetPtr=&/ThermalDMR_Fire_Close_Mech_03}}
 {id=WavePlayer_13, properties {soundWaveAssetPtr=&/ThermalDMR_Fire_Close_Mech_04}}
 {id=WavePlayer_14, properties {soundWaveAssetPtr=&/ThermalDMR_Fire_Close_Mech_05}}
 {id=WavePlayer_15, properties {soundWaveAssetPtr=&/ThermalDMR_Fire_Sub_03}}
 {id=WavePlayer_16, properties {soundWaveAssetPtr=&/ThermalDMR_Fire_Medium_01}}
 {id=WavePlayer_17, properties {soundWaveAssetPtr=&/ThermalDMR_Fire_Medium_02}}
 {id=WavePlayer_18, properties {soundWaveAssetPtr=&/ThermalDMR_Fire_Medium_03}}
 {id=WavePlayer_19, properties {soundWaveAssetPtr=&/ThermalDMR_Fire_Medium_04}}
 {id=WavePlayer_2, properties {soundWaveAssetPtr=&ThermalDMR_Fire_Close_A_01}}
 {id=WavePlayer_20, properties {soundWaveAssetPtr=&/ThermalDMR_Fire_Medium_05}}
 {id=WavePlayer_21, properties {soundWaveAssetPtr=&/ThermalDMR_Fire_Distant_01}}
 {id=WavePlayer_22, properties {soundWaveAssetPtr=&/ThermalDMR_Fire_Distant_02}}
 {id=WavePlayer_23, properties {soundWaveAssetPtr=&/ThermalDMR_Fire_Distant_03}}
 {id=WavePlayer_24, properties {soundWaveAssetPtr=&/ThermalDMR_Fire_Distant_04}}
 {id=WavePlayer_25, properties {soundWaveAssetPtr=&/ThermalDMR_Fire_Distant_05}}
 {id=WavePlayer_3, properties {soundWaveAssetPtr=&ThermalDMR_Fire_Close_A_02}}
 {id=WavePlayer_35, properties {soundWaveAssetPtr=&ThermalDMR_Fire_Close_Mech_01}}
 {id=WavePlayer_36, properties {soundWaveAssetPtr=&ThermalDMR_Fire_Close_Mech_02}}
 {id=WavePlayer_37, properties {soundWaveAssetPtr=&ThermalDMR_Fire_Close_Mech_03}}
 {id=WavePlayer_38, properties {soundWaveAssetPtr=&ThermalDMR_Fire_Close_Mech_04}}
 {id=WavePlayer_39, properties {soundWaveAssetPtr=&ThermalDMR_Fire_Close_Mech_05}}
 {id=WavePlayer_4, properties {soundWaveAssetPtr=&/ThermalDMR_Fire_Close_A_03}}
 {id=WavePlayer_40, properties {soundWaveAssetPtr=&ThermalDMR_Fire_Sub_03}}
 {id=WavePlayer_45, properties {soundWaveAssetPtr=&ThermalDMR_Fire_Medium_01}}
 {id=WavePlayer_46, properties {soundWaveAssetPtr=&ThermalDMR_Fire_Medium_02}}
 {id=WavePlayer_47, properties {soundWaveAssetPtr=&ThermalDMR_Fire_Medium_03}}
 {id=WavePlayer_48, properties {soundWaveAssetPtr=&ThermalDMR_Fire_Medium_04}}
 {id=WavePlayer_49, properties {soundWaveAssetPtr=&ThermalDMR_Fire_Medium_05}}
 {id=WavePlayer_5, properties {soundWaveAssetPtr=&/ThermalDMR_Fire_Close_A_04}}
 {id=WavePlayer_50, properties {soundWaveAssetPtr=&ThermalDMR_Fire_Distant_01}}
 {id=WavePlayer_51, properties {soundWaveAssetPtr=&ThermalDMR_Fire_Distant_02}}
 {id=WavePlayer_52, properties {soundWaveAssetPtr=&ThermalDMR_Fire_Distant_03}}
 {id=WavePlayer_53, properties {soundWaveAssetPtr=&ThermalDMR_Fire_Distant_04}}
 {id=WavePlayer_54, properties {soundWaveAssetPtr=&ThermalDMR_Fire_Distant_05}}
 {id=WavePlayer_6, properties {soundWaveAssetPtr=&ThermalDMR_Fire_Close_A_03}}
 {id=WavePlayer_7, properties {soundWaveAssetPtr=&/ThermalDMR_Fire_Close_A_05}}
 {id=WavePlayer_8, properties {soundWaveAssetPtr=&ThermalDMR_Fire_Close_A_04}}
 {id=WavePlayer_9, properties {soundWaveAssetPtr=&ThermalDMR_Fire_Close_A_05}}
]
properties {
 bDebug=true
 bExcludeFromRandomNodeBranchCulling=true
 maxDistance=27500
 soundClassObject=&Fort_WeaponRanged_Fire_3P_PSM
 volumeMultiplier=0.95
 concurrencyOverrides {bLimitToOwner=true, maxCount=3, resolutionRule=StopOldest}
 concurrencySet [&Weapon_Gunfire_Local_Conc, &Weapon_Gunfire_Global_Conc]
}
