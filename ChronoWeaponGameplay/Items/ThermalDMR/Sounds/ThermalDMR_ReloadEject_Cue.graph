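# Sound cue for the Thermal DMR magazine eject during reload, classed as weapon foley and using weapon reload attenuation.
# A mixer layers one of three random reload-start sounds with a quieter, delayed magazine impact.
# The impact picks randomly from two waves, passes through a delay of just over half a second, then a default modulator.
# The cue overrides concurrency and has its debug flag enabled.
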SoundCue /ChronoWeaponGameplay/Items/ThermalDMR/Sounds/ThermalDMR_ReloadEject_Cue
root=Mixer_0
nodes [
 {id=Delay_0, children [Random_2], properties {delayMax=0.65, delayMin=0.6}}
 {id=Mixer_0, children [Random_1,Modulator_2], properties {inputVolume [1,0.3]}}
 {id=Modulator_2, children [Delay_0]}
 {id=Random_1, children [WavePlayer_4,WavePlayer_5,WavePlayer_6], properties {weights [1,1,1]}}
 {id=Random_2, children [WavePlayer_7,WavePlayer_9], properties {weights [1,1]}}
 {id=WavePlayer_4, properties {soundWaveAssetPtr=&/ThermalDMR_ReloadStart_01}}
 {id=WavePlayer_5, properties {soundWaveAssetPtr=&/ThermalDMR_ReloadStart_02}}
 {id=WavePlayer_6, properties {soundWaveAssetPtr=&/ThermalDMR_ReloadStart_03}}
 {id=WavePlayer_7, properties {soundWaveAssetPtr=&/ThermalDMR_Magazine_Impact_01}}
 {id=WavePlayer_9, properties {soundWaveAssetPtr=&/ThermalDMR_Magazine_Impact_03}}
]
properties {
 attenuationSettings=&Fort_Weapon_Reload_Att
 bDebug=true
 bOverrideConcurrency=true
 maxDistance=2250
 soundClassObject=&Fort_WeaponFoley
 volumeMultiplier=0.551
}
