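# Sound cue for the insert step of the Thermal DMR weapon's reload.
# A Random node picks one of three equally weighted insert wave variations and feeds it into a root Modulator left at default settings.
# The cue plays as weapon foley at reduced volume, with reload attenuation and overridden concurrency.
# Its debug flag is switched on.
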SoundCue /ChronoWeaponGameplay/Items/ThermalDMR/Sounds/ThermalDMR_ReloadInsert_Cue
root=Modulator_0
nodes [
 {id=Modulator_0, children [Random_0]}
 {id=Random_0, children [WavePlayer_0,WavePlayer_1,WavePlayer_2], properties {weights [1,1,1]}}
 {id=WavePlayer_0, properties {soundWaveAssetPtr=&/ThermalDMR_ReloadInsert_01}}
 {id=WavePlayer_1, properties {soundWaveAssetPtr=&/ThermalDMR_ReloadInsert_02}}
 {id=WavePlayer_2, properties {soundWaveAssetPtr=&/ThermalDMR_ReloadInsert_03}}
]
properties {
 attenuationSettings=&Fort_Weapon_Reload_Att
 bDebug=true
 bOverrideConcurrency=true
 maxDistance=2250
 soundClassObject=&Fort_WeaponFoley
 volumeMultiplier=0.601
}
