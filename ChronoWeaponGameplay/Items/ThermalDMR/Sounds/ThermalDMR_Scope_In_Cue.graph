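# Sound cue for the Thermal DMR weapon's scope-in sound, classed as weapon foley.
# A Random node picks among four equally weighted wave variations and feeds a Modulator at the root for pitch and volume variation.
# The cue lowers overall volume and pitch, and its attenuation override disables both spatialization and distance falloff.
# Its concurrency override prevents new instances once the limit is reached.
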SoundCue /ChronoWeaponGameplay/Items/ThermalDMR/Sounds/ThermalDMR_Scope_In_Cue
root=Modulator_2
nodes [
 {id=Modulator_2, children [Random_0]}
 {id=Random_0, children [WavePlayer_0,WavePlayer_1,WavePlayer_2,WavePlayer_4], properties {weights [1,1,1,1]}}
 {id=WavePlayer_0, properties {soundWaveAssetPtr=&/ThermalDMR_Scope_In_01}}
 {id=WavePlayer_1, properties {soundWaveAssetPtr=&/ThermalDMR_Scope_In_02}}
 {id=WavePlayer_2, properties {soundWaveAssetPtr=&/ThermalDMR_Scope_In_03}}
 {id=WavePlayer_4, properties {soundWaveAssetPtr=&/ThermalDMR_Scope_In_04}}
]
properties {
 bOverrideAttenuation=true
 bOverrideConcurrency=true
 maxDistance=2097152
 pitchMultiplier=0.9
 soundClassObject=&Fort_WeaponFoley
 volumeMultiplier=0.301
 attenuationOverrides {bAttenuate=false, bSpatialize=false, falloffDistance=1500, nonSpatializedRadiusStart=200, attenuationShapeExtents {x=1200}}
 concurrencyOverrides {resolutionRule=PreventNew}
}
