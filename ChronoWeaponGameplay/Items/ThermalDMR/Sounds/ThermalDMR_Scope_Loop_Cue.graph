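# Sound cue for the looping noise of the Thermal DMR's scope, played quietly under the weapon foley sound class.
# A Random node picks one of two equally weighted looping noise waves, which passes through a fixed pitch-up Modulator and a short fixed Delay.
# Attenuation and concurrency are overridden on the cue, with concurrency set to prevent new instances once the limit is reached.
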SoundCue /ChronoWeaponGameplay/Items/ThermalDMR/Sounds/ThermalDMR_Scope_Loop_Cue
root=Delay_0
nodes [
 {id=Delay_0, children [Modulator_2], properties {delayMax=0.07, delayMin=0.07}}
 {id=Modulator_2, children [Random_4], properties {pitchMax=1.2, pitchMin=1.2}}
 {id=Random_4, children [WavePlayer_20, WavePlayer_22], properties {weights [1,1]}}
 {id=WavePlayer_20, properties {bLooping=true, soundWaveAssetPtr=&/ThermalDMR_Scope_Loop_Noise_01}}
 {id=WavePlayer_22, properties {bLooping=true, soundWaveAssetPtr=&/ThermalDMR_Scope_Loop_Noise_03}}
]
properties {
 bOverrideAttenuation=true
 bOverrideConcurrency=true
 maxDistance=2000
 soundClassObject=&Fort_WeaponFoley
 volumeMultiplier=0.125
 attenuationOverrides {falloffDistance=1500, nonSpatializedRadiusStart=400, stereoSpread=300, attenuationShapeExtents {x=500}}
 concurrencyOverrides {resolutionRule=PreventNew}
}
