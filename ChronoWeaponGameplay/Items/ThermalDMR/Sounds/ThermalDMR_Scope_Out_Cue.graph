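# Sound cue for the Thermal DMR scope-out sound, classed as weapon foley.
# A Random node picks one of four equally weighted wave variations and feeds it through a Modulator at the root.
# The cue plays at reduced volume and slightly lowered pitch, with overrides that turn off attenuation and spatialization.
# A concurrency override prevents new instances from starting once its limit is reached.
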SoundCue /ChronoWeaponGameplay/Items/ThermalDMR/Sounds/ThermalDMR_Scope_Out_Cue
root=Modulator_0
nodes [
 {id=Modulator_0, children [Random_0]}
 {id=Random_0, children [WavePlayer_1,WavePlayer_2,WavePlayer_3,WavePlayer_4], properties {weights [1,1,1,1]}}
 {id=WavePlayer_1, properties {soundWaveAssetPtr=&/ThermalDMR_Scope_Out_01}}
 {id=WavePlayer_2, properties {soundWaveAssetPtr=&/ThermalDMR_Scope_Out_02}}
 {id=WavePlayer_3, properties {soundWaveAssetPtr=&/ThermalDMR_Scope_Out_03}}
 {id=WavePlayer_4, properties {soundWaveAssetPtr=&/ThermalDMR_Scope_Out_04}}
]
properties {
 bOverrideAttenuation=true
 bOverrideConcurrency=true
 maxDistance=2097152
 pitchMultiplier=0.9
 soundClassObject=&Fort_WeaponFoley
 volumeMultiplier=0.251
 attenuationOverrides {bAttenuate=false, bSpatialize=false, falloffDistance=1500, nonSpatializedRadiusStart=200, attenuationShapeExtents {x=1200}}
 concurrencyOverrides {resolutionRule=PreventNew}
}
