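# Sound cue for switching off the thermal view on the Thermal DMR scope.
# A Random node picks between two equally weighted wave variations.
# It plays at reduced volume in the weapon foley sound class, with attenuation and spatialization overridden off so it is heard as flat 2D audio.
# A concurrency override blocks new instances when the limit is reached instead of cutting off playing ones.
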SoundCue /ChronoWeaponGameplay/Items/ThermalDMR/Sounds/ThermalDMR_Scope_Thermal_Deactivate_Cue
root=Random_0
nodes [
 {id=Random_0, children [WavePlayer_1,WavePlayer_2], properties {weights [1,1]}}
 {id=WavePlayer_1, properties {soundWaveAssetPtr=&/ThermalDMR_Scope_Thermal_Deactivate_A_01}}
 {id=WavePlayer_2, properties {soundWaveAssetPtr=&/ThermalDMR_Scope_Thermal_Deactivate_A_02}}
]
properties {
 bOverrideAttenuation=true
 bOverrideConcurrency=true
 maxDistance=2097152
 soundClassObject=&Fort_WeaponFoley
 volumeMultiplier=0.226
 attenuationOverrides {bAttenuate=false, bSpatialize=false, falloffDistance=1500, nonSpatializedRadiusStart=200, attenuationShapeExtents {x=1200}}
 concurrencyOverrides {resolutionRule=PreventNew}
}
