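# Sound cue for the thermal scope of the ThermalDMR weapon, combining an activation sound with a sustained scope loop.
# A top mixer plays one randomly chosen activation one-shot, through a modulator with fixed pitch, alongside a sub-mixer of three looping layers: two beeps and a quieter noise bed.
# It uses the weapon foley sound class, overrides attenuation, and sets concurrency to prevent new instances at the limit.
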SoundCue /ChronoWeaponGameplay/Items/ThermalDMR/Sounds/ThermalDMR_Scope_Thermal_Loop_Cue
root=Mixer_0
nodes [
 {id=Mixer_0, children [Modulator_1,Mixer_1], properties {inputVolume [1,1.2]}}
 {id=Mixer_1, children [WavePlayer_4,WavePlayer_5,WavePlayer_0], properties {inputVolume [0.8,0.8,0.5]}}
 {id=Modulator_1, children [Random_3], properties {pitchMax=1, pitchMin=1}}
 {id=Random_3, children [WavePlayer_10, WavePlayer_14, WavePlayer_15], properties {weights [1,1,1]}}
 {id=WavePlayer_0, properties {bLooping=true, soundWaveAssetPtr=&/ThermalDMR_Scope_Loop_Noise_01}}
 {id=WavePlayer_10, properties {soundWaveAssetPtr=&/ThermalDMR_Scope_Thermal_Activate_A_01}}
 {id=WavePlayer_14, properties {soundWaveAssetPtr=&/ThermalDMR_Scope_Thermal_Activate_C_01}}
 {id=WavePlayer_15, properties {soundWaveAssetPtr=&/ThermalDMR_Scope_Thermal_Activate_C_02}}
 {id=WavePlayer_4, properties {bLooping=true, soundWaveAssetPtr=&/ThermalDMR_Scope_Loop_Beeps_01}}
 {id=WavePlayer_5, properties {bLooping=true, soundWaveAssetPtr=&/ThermalDMR_Scope_Loop_Beeps_02}}
]
properties {
 bOverrideAttenuation=true
 bOverrideConcurrency=true
 maxDistance=2000
 soundClassObject=&Fort_WeaponFoley
 volumeMultiplier=0.226
 attenuationOverrides {falloffDistance=1500, nonSpatializedRadiusStart=400, stereoSpread=300, attenuationShapeExtents {x=500}}
 concurrencyOverrides {resolutionRule=PreventNew}
}
